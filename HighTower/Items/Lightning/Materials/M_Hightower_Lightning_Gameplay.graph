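# Skeletal-mesh gameplay material for the HighTower Lightning item, layering an animated glow over albedo, mask and normal maps that default to HighTower Date textures.
# Fresnel rims drive glow color and world-aligned caustics panned in pre-skinned space, mirrored across X.
# The glow fades out with height, blends with albedo and scales by GlowActive.
# Eye-adaptation-corrected emissive joins the surface attributes, then passes through TechArt ClearCoat and MetalLUT functions.
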Material /HighTower/Items/Lightning/Materials/M_Hightower_Lightning_Gameplay
graph {
 properties {
  bAutomaticallySetUsageInEditor=false
  bUseMaterialAttributes=true
  bUsedWithMeshDeformer=true
  bUsedWithSkeletalMesh=true
  usedShadingModels=Default Lit
 }
 nodes [
  n1 = NamedRerouteUsage()
  n2 = TextureSampleParameter2D() {parameterName=Albedo, texture=/HighTower/Items/Date/Combo/Textures/T_HighTowerDate_D}
  n3 = TextureSampleParameter2D() {parameterName=Masks, samplerType=SAMPLERTYPE_Masks, texture=/HighTower/Items/Date/Combo/Textures/T_HighTowerDate_S}
  n4 = ScalarParameter() {defaultValue=0.8, parameterName=EmissiveAlbedoBlend}
  n5 = Transform(None=NamedRerouteUsage())
  n6 = MaterialFunctionCall(
   Brightness=ScalarParameter() {defaultValue=0.6, parameterName=FresnelBrightness}
   Exponent=ScalarParameter() {defaultValue=2.5, parameterName=FresnelExponent}
   Vertex / Pixel Normal=@n5
  ) {materialFunction=&MF_Fresnel_FX}
  n7 = Max(
   A=@n6.Fresnel Inverted
   B=MaterialFunctionCall!(
    Thickness=ScalarParameter() {defaultValue=0.5, parameterName=FresnelThickness}
    NormalWS=@n5
    PowerExponent=ScalarParameter() {defaultValue=1.5, parameterName=NormalizeFresnelExponent}
   ) {materialFunction=/Game/Weapons/Prototype/ObjectMover/Materials/MF_NormalizedFresnel}.PowerInverted
  )
  n8 = VertexInterpolator(VS=LocalPosition() {localOrigin=InstancePreSkinning}.XYZ)
  n9 = Time()
  n10 = ComponentMask(None=@n8.PS) {r=true}
  n11 = Constant() {r=1}
  n12 = Multiply(
   A=VectorParameter() {parameterName=PanSpeedPSLP, defaultValue=<rgba -5 0 -20 0>}.RGB
   B=MaterialFunctionCall(R=Sign(None=@n10), G=@n11, B=@n11) {materialFunction=&AppendMany}.RGB
  )
  n13 = MaterialFunctionCall(
   TextureObject=TextureObject() {samplerType=SAMPLERTYPE_LinearColor, texture=/Game/Weapons/FORT_Melee/Pickaxe_RoyalDsuk/Textures/T_Pickaxe_RoyalDusk_Caustics}
   TextureSize=VectorParameter() {parameterName=CausticsScale, defaultValue=<rgba 20 20 40 0>}.RGB
   WorldPosition=Subtract(A=@n8.PS, B=Multiply(A=@n9, B=@n12))
   World Space Normal=VertexInterpolator(VS=PreSkinnedNormal()).PS
  ) {materialFunction=&WorldAlignedTexture}
  n14 = VectorParameter() {parameterName=CausticColor, defaultValue=<rgba 75.6036 100 14.0869 1>}
  n15 = Add(
   A=Multiply(A=Saturate(None=@n7), B=VectorParameter() {parameterName=Emissive Glow, defaultValue=<rgba 0.5029 8 0.9612 1>}.RGB)
   B=Multiply(A=@n6.Fresnel Inverted, B=Multiply(A=@n13.XYZ Texture, B=@n14.RGB))
  )
  n16 = Add(A=ComponentMask(None=@n8.PS) {b=true}) {constB=-80}
  n17 = OneMinus(None=Saturate(None=Divide(A=@n16) {constB=28}))
  n18 = Multiply(
   A=Multiply(A=LinearInterpolate(B=@n2.RGB, Alpha=@n4) {constA=1}, B=Multiply(A=@n15, B=@n17))
   B=ScalarParameter() {defaultValue=1, parameterName=GlowActive}
  )
  n19 = GetMaterialAttributes(None=MaterialFunctionCall!(Material=SetMaterialAttributes!(MaterialAttributes), Emissive UVs, Fresnel, Alpha) {materialFunction=/Game/Global/Materials/Character_Parent_2023/Functions/MF_CharacterParent_2023_Emissive}.Result) {attributeGetTypes [#1]}
  n20 = SetMaterialAttributes!(
   MaterialAttributes
   Base Color=@n2.RGB
   Metallic=@n3.B
   Specular=@n3.B
   Roughness=@n3.B
   Emissive Color=EyeAdaptationInverse(LightValueInput=Add(A=@n18, B=@n19.Emissive Color)).EyeAdaptationInverse
   Normal=NamedRerouteDeclaration(None=TextureSampleParameter2D() {parameterName=Normals, samplerType=SAMPLERTYPE_Normal, texture=/HighTower/Items/Date/Combo/Textures/T_HighTowerDate_N}.RGB) {name=NormalMap, nodeColor=<rgba 0.9922 0 1 1>}
  ) {attributeSetTypes [
    #2
    #3
    #4
    #5
    #1
    #6
   ]}
  MaterialFunctionCall!(
   Alpha=@n1
   Materials=MaterialFunctionCall!(Material=@n20, Alpha=@n1) {materialFunction=/Game/Global/MaterialFunctions/TechArt/MF_ClearCoat_TechArt}.Result
  ) {materialFunction=/Game/Global/MaterialFunctions/TechArt/MF_MetalLUT_TechArt}
  NamedRerouteDeclaration(None=TextureSampleParameter2D() {parameterName=TechArtMask, samplerType=SAMPLERTYPE_Masks, texture=/HighTower/Items/Date/Combo/Textures/T_HighTowerDate_FX}.RGB) {name=TechArtMask, nodeColor=<rgba 0.6172 1 0 1>}
 ]
}
